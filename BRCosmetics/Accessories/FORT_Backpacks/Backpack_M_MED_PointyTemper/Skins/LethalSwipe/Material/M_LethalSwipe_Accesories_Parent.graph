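# Parent character material for LethalSwipe backpack accessories, adding a slurp-and-bubble look to green-channel Tech Art Mask regions.
# MF_CharacterParent_2023_PreFX feeds MF_Slurp_2023, then LethalSwipe bubbles blend into base color and emissive.
# A colored Fresnel joins the emissive, which gets a time-of-day tint.
# Named reroutes supply skin and Fresnel-darkening masks for lerps before PostFX, and shading model comes from the material expression.
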Material /BRCosmetics/Accessories/FORT_Backpacks/Backpack_M_MED_PointyTemper/Skins/LethalSwipe/Material/M_LethalSwipe_Accesories_Parent
graph {
 properties {
  bUseMaterialAttributes=true
  bUsedWithMeshDeformer=true
  bUsedWithMorphTargets=true
  bUsedWithSkeletalMesh=true
  pixelDepthOffsetMode=PDOM_Legacy
  shadingModel=MSM_FromMaterialExpression
  usedShadingModels=Default Lit | Subsurface | Clear Coat | Subsurface Profile | Cloth
 }
 nodes [
  n1 = MaterialFunctionCall() {materialFunction=&MF_CharacterParent_2023_PreFX}
  n2 = ComponentMask(None=@n1.Tech Art Mask) {g=true}
  n3 = GetMaterialAttributes(None=MaterialFunctionCall!(SlurpMask=@n2, AO, BubblesMask, SlurpNormalDistortionMask, Pre Skinned Position, Material=@n1.Result) {materialFunction=/Game/Global/Materials/Character_Parent_2023/Slurp_Parent_2023/MF_Slurp_2023}.Result) {attributeGetTypes [#1, #2]}
  n4 = NamedRerouteUsage()
  n5 = Constant4Vector() {constant=<rgba 0 0.1 0 1>}
  n6 = MaterialFunctionCall!(Background Parallax Offset, Mask=@n2, Foreground Parallax Offset) {materialFunction=/BRCosmetics/Characters/Player/Female/Medium/Bodies/F_MED_LethalSwipe/Materials/MF_LethalSwipe_Bubbles}
  n7 = LinearInterpolate(A=@n5.RGB, B=LinearInterpolate(A=@n3.Base Color, B=@n6.Result, Alpha=@n6.BubblesMask), Alpha=NamedRerouteUsage())
  n8 = Add(A=@n3.Emissive Color, B=Multiply(A=@n6.Result, B=@n6.BubblesMask))
  n9 = Multiply(
   A=Fresnel(ExponentIn=ScalarParameter() {defaultValue=3, parameterName=Fresnel Exp})
   B=VectorParameter() {parameterName=Fresnel Color, defaultValue=<rgba 1 0.8333 0 1>}.RGB
  )
  n10 = Constant3Vector() {constant=<rgba 1 1 1 0>}
  n11 = Constant3Vector() {constant=<rgba 0.5 0.5833 1 0>}
  n12 = LinearInterpolate(
   A=@n5.RGB
   B=Multiply(
    A=Add(A=@n8, B=@n9)
    B=MaterialFunctionCall(Color_Morning=@n10, Color_Day=@n10, Color_Evening=@n10, Color_Night=@n11) {materialFunction=&Color_Time_of_Day}.Result
   )
   Alpha=NamedRerouteUsage()
  )
  MaterialFunctionCall(
   Material=SetMaterialAttributes!(
    MaterialAttributes=@n3.MaterialAttributes
    Base Color=LinearInterpolate(A=@n3.Base Color, B=@n7, Alpha=@n4)
    Emissive Color=LinearInterpolate(A=@n3.Emissive Color, B=@n12, Alpha=@n4)
   ) {attributeSetTypes [#1, #2]}
   M=@n1.M
   Pre Skinned Local Position=@n1.M
   Distance Blend=@n1.M
   Fresnel=@n1.M
  ) {materialFunction=&MF_CharacterParent_2023_PostFX}
  n13 = Fresnel(ExponentIn=ScalarParameter() {defaultValue=1, parameterName=Darkening_FresnelExp})
  n14 = OneMinus(None=ScalarParameter() {defaultValue=0.5, parameterName=Darkening_FresnelIntensity})
  NamedRerouteDeclaration(None=Saturate(None=Add(A=@n13, B=@n14))) {name=FresnelMask, nodeColor=<rgba 1 0.9141 0 1>}
  NamedRerouteDeclaration(None=@n2) {name=SkinMask, nodeColor=<rgba 0.0547 1 0 1>}
 ]
}
